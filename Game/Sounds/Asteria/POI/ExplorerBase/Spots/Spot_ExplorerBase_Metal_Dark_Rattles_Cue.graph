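# Sound cue for dark metal rattle ambient spot sounds at the Asteria Explorer Base POI, reusing five Hollow Base cave rattle waves from Artemis.
# The root randomly picks one of two envelopers with different volume fade-out shapes, both also applying a pitch curve.
# Both share a randomly down-pitched modulator over an equally weighted random wave choice.
# Playback uses distant-creak one-shot attenuation and the outdoor ambient sound class, with bus and submix sends disabled.
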SoundCue /Game/Sounds/Asteria/POI/ExplorerBase/Spots/Spot_ExplorerBase_Metal_Dark_Rattles_Cue
root=Random_1
nodes [
 {
  id=Enveloper_0
  children [Modulator_0]
  properties {
   pitchCurve {editorCurveData {keys [{time=0}, <at 10 0.95>]}}
   volumeCurve {editorCurveData {keys [<at 0 1>, {arriveTangent=-0.1538, interpMode=RCIM_Cubic, leaveTangent=-0.1538, time=3, value=1}, <at 6.5 0 RCIM_Cubic>]}}
  }
 }
 {
  id=Enveloper_1
  children [Modulator_0]
  properties {
   pitchCurve {editorCurveData {keys [{time=0}, <at 10 0.95>]}}
   volumeCurve {editorCurveData {keys [<at 0 1>, {interpMode=RCIM_Cubic, leaveTangent=-0.1538, time=1, value=1}, <at 8 0 RCIM_Cubic>]}}
  }
 }
 {id=Modulator_0, children [Random_0], properties {pitchMax=1, pitchMin=0.7}}
 {id=Random_0, children [WavePlayer_0,WavePlayer_1,WavePlayer_2,WavePlayer_3,WavePlayer_4], properties {weights [1,1,1,1,1]}}
 {id=Random_1, children [Enveloper_1,Enveloper_0], properties {weights [1,1]}}
 {id=WavePlayer_0, properties {soundWaveAssetPtr=/Game/Sounds/Artemis/POI/Hollow_Base/Spots/Spot_Cave_Metal_Dark_Rattle_01}}
 {id=WavePlayer_1, properties {soundWaveAssetPtr=/Game/Sounds/Artemis/POI/Hollow_Base/Spots/Spot_Cave_Metal_Dark_Rattle_02}}
 {id=WavePlayer_2, properties {soundWaveAssetPtr=/Game/Sounds/Artemis/POI/Hollow_Base/Spots/Spot_Cave_Metal_Dark_Rattle_03}}
 {id=WavePlayer_3, properties {soundWaveAssetPtr=/Game/Sounds/Artemis/POI/Hollow_Base/Spots/Spot_Cave_Metal_Dark_Rattle_04}}
 {id=WavePlayer_4, properties {soundWaveAssetPtr=/Game/Sounds/Artemis/POI/Hollow_Base/Spots/Spot_Cave_Metal_Dark_Rattle_05}}
]
properties {
 attenuationSettings=/Game/Sounds/Attenuation/Fort_TileAmbient_OneShot_Distant_Creak_Att
 bEnableBusSends=false
 bEnableSubmixSends=false
 maxDistance=8200
 soundClassObject=&Fort_Ambient_Outdoors_Athena
 volumeMultiplier=0.501
 attenuationOverrides {falloffDistance=6000, stereoSpread=10000, attenuationShapeExtents {x=2000}}
}
