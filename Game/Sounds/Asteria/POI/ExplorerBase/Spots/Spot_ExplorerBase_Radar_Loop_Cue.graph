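# Looping ambient spot cue for a radar in the Asteria Explorer Base POI, reusing an Apollo radar dish glow loop.
# A looping wave player feeds a modulator that fixes a lowered pitch, then a mixer that boosts volume.
# The quality-level root routes its first two levels to that chain and leaves the other two empty.
# It uses the smaller radar dish attenuation asset and limits concurrency to one instance.
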SoundCue /Game/Sounds/Asteria/POI/ExplorerBase/Spots/Spot_ExplorerBase_Radar_Loop_Cue
root=QualityLevel_0
nodes [
 {id=Mixer_0, children [Modulator_0], properties {inputVolume [1.65]}}
 {id=Modulator_0, children [WavePlayer_2], properties {pitchMax=0.8, pitchMin=0.8, volumeMax=1, volumeMin=1}}
 {id=QualityLevel_0, children [Mixer_0,Mixer_0,null,null]}
 {id=WavePlayer_2, properties {bLooping=true, soundWaveAssetPtr=/Game/Sounds/Apollo/Biomes/RadarBases/Amb_RadarDish_Glow_Loop_03}}
]
properties {
 attenuationSettings=&/Spot_ExplorerBase_Radar_Dish_Smaller_Att
 bOverrideConcurrency=true
 maxDistance=1500
 soundClassObject=&Fort_Ambient_Outdoors_Athena
 volumeMultiplier=0.376
 attenuationOverrides {
  bAttenuateWithLPF=true
  bEnableReverbSend=false
  falloffDistance=7000
  lPFFrequencyAtMax=500
  lPFRadiusMax=5000
  lPFRadiusMin=2000
  stereoSpread=5000
  attenuationShapeExtents {x=1000}
 }
 concurrencyOverrides {maxCount=1}
}
