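# Sound cue for the "Chime B" ambient spot sound at the Asteria Explorer Base POI.
# A Random node picks among thirteen equally weighted Chime B waves borrowed from the Rainforest Village mystery set.
# A root Modulator varies pitch and volume, with pitch able to drop well below normal for variety.
# It plays at reduced volume in the outdoor ambient sound class, referencing a shared magic-presence attenuation asset.
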SoundCue /Game/Sounds/Asteria/POI/ExplorerBase/Spots/Spot_Explorer_Base_Chime_B_Cue
root=Modulator_0
nodes [
 {id=Modulator_0, children [Random_0], properties {pitchMin=0.65, volumeMin=0.8}}
 {
  id=Random_0
  children [
   WavePlayer_9
   WavePlayer_10
   WavePlayer_11
   WavePlayer_12
   WavePlayer_13
   WavePlayer_14
   WavePlayer_15
   WavePlayer_16
   WavePlayer_17
   WavePlayer_18
   WavePlayer_19
   WavePlayer_20
   WavePlayer_21
  ]
  properties {weights [1,1,1,1,1,1,1,1,1,1,1,1,1]}
 }
 {
  id=WavePlayer_10
  properties {soundWaveAssetPtr=/Game/Sounds/Asteria/POI/RainForest_Village/Spots/Spot_Rainforest_Mystery_Deconstructed_Chime_B_01}
 }
 {
  id=WavePlayer_11
  properties {soundWaveAssetPtr=/Game/Sounds/Asteria/POI/RainForest_Village/Spots/Spot_Rainforest_Mystery_Deconstructed_Chime_B_02}
 }
 {
  id=WavePlayer_12
  properties {soundWaveAssetPtr=/Game/Sounds/Asteria/POI/RainForest_Village/Spots/Spot_Rainforest_Mystery_Deconstructed_Chime_B_03}
 }
 {
  id=WavePlayer_13
  properties {soundWaveAssetPtr=/Game/Sounds/Asteria/POI/RainForest_Village/Spots/Spot_Rainforest_Mystery_Deconstructed_Chime_B_04}
 }
 {
  id=WavePlayer_14
  properties {soundWaveAssetPtr=/Game/Sounds/Asteria/POI/RainForest_Village/Spots/Spot_Rainforest_Mystery_Deconstructed_Chime_B_05}
 }
 {
  id=WavePlayer_15
  properties {soundWaveAssetPtr=/Game/Sounds/Asteria/POI/RainForest_Village/Spots/Spot_Rainforest_Mystery_Deconstructed_Chime_B_06}
 }
 {
  id=WavePlayer_16
  properties {soundWaveAssetPtr=/Game/Sounds/Asteria/POI/RainForest_Village/Spots/Spot_Rainforest_Mystery_Deconstructed_Chime_B_07}
 }
 {
  id=WavePlayer_17
  properties {soundWaveAssetPtr=/Game/Sounds/Asteria/POI/RainForest_Village/Spots/Spot_Rainforest_Mystery_Deconstructed_Chime_B_08}
 }
 {
  id=WavePlayer_18
  properties {soundWaveAssetPtr=/Game/Sounds/Asteria/POI/RainForest_Village/Spots/Spot_Rainforest_Mystery_Deconstructed_Chime_B_09}
 }
 {
  id=WavePlayer_19
  properties {soundWaveAssetPtr=/Game/Sounds/Asteria/POI/RainForest_Village/Spots/Spot_Rainforest_Mystery_Deconstructed_Chime_B_10}
 }
 {
  id=WavePlayer_20
  properties {soundWaveAssetPtr=/Game/Sounds/Asteria/POI/RainForest_Village/Spots/Spot_Rainforest_Mystery_Deconstructed_Chime_B_11}
 }
 {
  id=WavePlayer_21
  properties {soundWaveAssetPtr=/Game/Sounds/Asteria/POI/RainForest_Village/Spots/Spot_Rainforest_Mystery_Deconstructed_Chime_B_12}
 }
 {
  id=WavePlayer_9
  properties {soundWaveAssetPtr=/Game/Sounds/Asteria/POI/RainForest_Village/Spots/Spot_Rainforest_Mystery_Deconstructed_Chime_B_13}
 }
]
properties {
 attenuationSettings=/Game/Sounds/Attenuation/Fort_TileAmbient_Magic_Pressence_Att
 maxDistance=10500
 soundClassObject=&Fort_Ambient_Outdoors_Athena
 volumeMultiplier=0.25
 attenuationOverrides {bAttenuate=false}
}
