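# Sound cue for an endlessly looping interior ambience spot of metal roof stress sounds at the Explorer Base POI.
# A Looping node repeats a randomized Delay and Modulator over an equally weighted Random pick from nineteen stress waves, giving irregular one-shots.
# It plays in the indoor ambient sound class at reduced volume, with overridden box-shaped, non-spatialized attenuation in place of the referenced stairwell wind attenuation asset.
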SoundCue /Game/Sounds/Asteria/POI/ExplorerBase/Spots/Spot_Explorer_Base_Metal_Roof_Int_Loop_Cue
root=Looping_0
nodes [
 {id=Delay_0, children [Modulator_1], properties {delayMax=4, delayMin=0.3}}
 {id=Looping_0, children [Delay_0]}
 {id=Modulator_1, children [Random_1]}
 {
  id=Random_1
  children [
   WavePlayer_0
   WavePlayer_1
   WavePlayer_2
   WavePlayer_3
   WavePlayer_4
   WavePlayer_5
   WavePlayer_6
   WavePlayer_17
   WavePlayer_18
   WavePlayer_19
   WavePlayer_20
   WavePlayer_21
   WavePlayer_22
   WavePlayer_23
   WavePlayer_24
   WavePlayer_25
   WavePlayer_26
   WavePlayer_27
   WavePlayer_28
  ]
  properties {weights [1,1,1,1,1,1,1,1,1,1,1,1,1,1,1,1,1,1,1]}
 }
 {id=WavePlayer_0, properties {soundWaveAssetPtr=&/Spot_Metal_Roof_Stress_Int_01}}
 {id=WavePlayer_1, properties {soundWaveAssetPtr=&/Spot_Metal_Roof_Stress_Int_02}}
 {id=WavePlayer_17, properties {soundWaveAssetPtr=&/Spot_Metal_Roof_Stress_Int_08}}
 {id=WavePlayer_18, properties {soundWaveAssetPtr=&/Spot_Metal_Roof_Stress_Int_09}}
 {id=WavePlayer_19, properties {soundWaveAssetPtr=&/Spot_Metal_Roof_Stress_Int_10}}
 {id=WavePlayer_2, properties {soundWaveAssetPtr=&/Spot_Metal_Roof_Stress_Int_03}}
 {id=WavePlayer_20, properties {soundWaveAssetPtr=&/Spot_Metal_Roof_Stress_Int_11}}
 {id=WavePlayer_21, properties {soundWaveAssetPtr=&/Spot_Metal_Roof_Stress_Int_12}}
 {id=WavePlayer_22, properties {soundWaveAssetPtr=&/Spot_Metal_Roof_Stress_Int_13}}
 {id=WavePlayer_23, properties {soundWaveAssetPtr=&/Spot_Metal_Roof_Stress_Int_14}}
 {id=WavePlayer_24, properties {soundWaveAssetPtr=&/Spot_Metal_Roof_Stress_Int_15}}
 {id=WavePlayer_25, properties {soundWaveAssetPtr=&/Spot_Metal_Roof_Stress_Int_16}}
 {id=WavePlayer_26, properties {soundWaveAssetPtr=&/Spot_Metal_Roof_Stress_Int_17}}
 {id=WavePlayer_27, properties {soundWaveAssetPtr=&/Spot_Metal_Roof_Stress_Int_18}}
 {id=WavePlayer_28, properties {soundWaveAssetPtr=&/Spot_Metal_Roof_Stress_Int_19}}
 {id=WavePlayer_3, properties {soundWaveAssetPtr=&/Spot_Metal_Roof_Stress_Int_04}}
 {id=WavePlayer_4, properties {soundWaveAssetPtr=&/Spot_Metal_Roof_Stress_Int_05}}
 {id=WavePlayer_5, properties {soundWaveAssetPtr=&/Spot_Metal_Roof_Stress_Int_06}}
 {id=WavePlayer_6, properties {soundWaveAssetPtr=&/Spot_Metal_Roof_Stress_Int_07}}
]
properties {
 attenuationSettings=/Game/Sounds/Artemis/POI/Seven_Workshop/Spot_Interior_Stairwell_Wind_Att
 bOverrideAttenuation=true
 maxDistance=1000
 soundClassObject=&Fort_Ambient_Indoors_Athena
 volumeMultiplier=0.501
 attenuationOverrides {attenuationShape=Box, bSpatialize=false, falloffDistance=200, attenuationShapeExtents=<600 800 400>}
}
