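# Sound cue for a distant wood call ambient spot sound at the Asteria Explorer Base POI, reusing Rainforest Village recordings.
# A top random node picks evenly between two nested random pools, the standard calls and the B variants, each equally weighted within.
# A modulator at the root varies each play, with a widened downward pitch range.
# It uses a closer one-shot ambient attenuation asset, an interior-ducking ambient sound class and boosted volume.
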SoundCue /Game/Sounds/Asteria/POI/ExplorerBase/Spots/Spot_Explorer_Base_Wood_Call_01_Cue
root=Modulator_0
nodes [
 {id=Modulator_0, children [Random_4], properties {pitchMin=0.8}}
 {
  id=Random_0
  children [WavePlayer_17, WavePlayer_18, WavePlayer_19, WavePlayer_20, WavePlayer_21, WavePlayer_22]
  properties {weights [1,1,1,1,1,1]}
 }
 {
  id=Random_2
  children [WavePlayer_23, WavePlayer_24, WavePlayer_25, WavePlayer_26, WavePlayer_27, WavePlayer_28, WavePlayer_29, WavePlayer_30]
  properties {weights [1,1,1,1,1,1,1,1]}
 }
 {id=Random_4, children [Random_2,Random_0], properties {weights [1,1]}}
 {
  id=WavePlayer_17
  properties {soundWaveAssetPtr=/Game/Sounds/Asteria/POI/RainForest_Village/Spots/Spot_Rainforest_Village_Distant_Wood_Call_B_1}
 }
 {
  id=WavePlayer_18
  properties {soundWaveAssetPtr=/Game/Sounds/Asteria/POI/RainForest_Village/Spots/Spot_Rainforest_Village_Distant_Wood_Call_B_2}
 }
 {
  id=WavePlayer_19
  properties {soundWaveAssetPtr=/Game/Sounds/Asteria/POI/RainForest_Village/Spots/Spot_Rainforest_Village_Distant_Wood_Call_B_3}
 }
 {
  id=WavePlayer_20
  properties {soundWaveAssetPtr=/Game/Sounds/Asteria/POI/RainForest_Village/Spots/Spot_Rainforest_Village_Distant_Wood_Call_B_4}
 }
 {
  id=WavePlayer_21
  properties {soundWaveAssetPtr=/Game/Sounds/Asteria/POI/RainForest_Village/Spots/Spot_Rainforest_Village_Distant_Wood_Call_B_5}
 }
 {
  id=WavePlayer_22
  properties {soundWaveAssetPtr=/Game/Sounds/Asteria/POI/RainForest_Village/Spots/Spot_Rainforest_Village_Distant_Wood_Call_B_6}
 }
 {
  id=WavePlayer_23
  properties {soundWaveAssetPtr=/Game/Sounds/Asteria/POI/RainForest_Village/Spots/Spot_Rainforest_Village_Distant_Wood_Call_8}
 }
 {
  id=WavePlayer_24
  properties {soundWaveAssetPtr=/Game/Sounds/Asteria/POI/RainForest_Village/Spots/Spot_Rainforest_Village_Distant_Wood_Call_1}
 }
 {
  id=WavePlayer_25
  properties {soundWaveAssetPtr=/Game/Sounds/Asteria/POI/RainForest_Village/Spots/Spot_Rainforest_Village_Distant_Wood_Call_2}
 }
 {
  id=WavePlayer_26
  properties {soundWaveAssetPtr=/Game/Sounds/Asteria/POI/RainForest_Village/Spots/Spot_Rainforest_Village_Distant_Wood_Call_3}
 }
 {
  id=WavePlayer_27
  properties {soundWaveAssetPtr=/Game/Sounds/Asteria/POI/RainForest_Village/Spots/Spot_Rainforest_Village_Distant_Wood_Call_4}
 }
 {
  id=WavePlayer_28
  properties {soundWaveAssetPtr=/Game/Sounds/Asteria/POI/RainForest_Village/Spots/Spot_Rainforest_Village_Distant_Wood_Call_5}
 }
 {
  id=WavePlayer_29
  properties {soundWaveAssetPtr=/Game/Sounds/Asteria/POI/RainForest_Village/Spots/Spot_Rainforest_Village_Distant_Wood_Call_6}
 }
 {
  id=WavePlayer_30
  properties {soundWaveAssetPtr=/Game/Sounds/Asteria/POI/RainForest_Village/Spots/Spot_Rainforest_Village_Distant_Wood_Call_7}
 }
]
properties {
 attenuationSettings=/Game/Sounds/Attenuation/Fort_TileAmbient_OneShot_Att_Closer
 maxDistance=8500
 soundClassObject=/Game/Sounds/FortSoundClassesAndModes/Fort_Ambient_InteriorSpace_Duck_TileMix_PSM
 volumeMultiplier=1.4
 attenuationOverrides {bAttenuate=false}
}
