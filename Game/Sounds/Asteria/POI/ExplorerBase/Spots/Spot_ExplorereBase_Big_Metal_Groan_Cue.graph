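# Sound cue for a big metal groan ambient spot at the Asteria Explorer Base POI, reusing Hollow Base distant coaster rattle-swell waves from Artemis.
# A Random node picks one of four equally weighted waves, then a Modulator varies pitch around normal.
# The Quality Level root feeds this chain to its first two slots and leaves the other two empty.
# It uses outdoor ambient class, shared distant-rattle attenuation and single-voice concurrency.
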SoundCue /Game/Sounds/Asteria/POI/ExplorerBase/Spots/Spot_ExplorereBase_Big_Metal_Groan_Cue
root=QualityLevel_1
nodes [
 {id=Modulator_2, children [Random_0], properties {pitchMax=1.2, pitchMin=0.8}}
 {id=QualityLevel_1, children [Modulator_2,Modulator_2,null,null]}
 {id=Random_0, children [WavePlayer_0,WavePlayer_1,WavePlayer_2,WavePlayer_3], properties {weights [1,1,1,1]}}
 {
  id=WavePlayer_0
  properties {soundWaveAssetPtr=/Game/Sounds/Artemis/POI/Hollow_Base_Takover/Spots/Spot_HollowBase_Coaster_Distant_Metal_Rattle_Swell_01}
 }
 {
  id=WavePlayer_1
  properties {soundWaveAssetPtr=/Game/Sounds/Artemis/POI/Hollow_Base_Takover/Spots/Spot_HollowBase_Coaster_Distant_Metal_Rattle_Swell_02}
 }
 {
  id=WavePlayer_2
  properties {soundWaveAssetPtr=/Game/Sounds/Artemis/POI/Hollow_Base_Takover/Spots/Spot_HollowBase_Coaster_Distant_Metal_Rattle_Swell_03}
 }
 {
  id=WavePlayer_3
  properties {soundWaveAssetPtr=/Game/Sounds/Artemis/POI/Hollow_Base_Takover/Spots/Spot_HollowBase_Coaster_Distant_Metal_Rattle_Swell_04}
 }
]
properties {
 attenuationSettings=/Game/Sounds/Artemis/POI/Hollow_Base_Takover/Spots/Spot_Metal_Coaster_Rattles_Distant_Att
 bOverrideConcurrency=true
 maxDistance=7865.4736
 soundClassObject=&Fort_Ambient_Outdoors_Athena
 volumeMultiplier=0.601
 attenuationOverrides {falloffDistance=1600}
 concurrencyOverrides {maxCount=1, voiceStealReleaseTime=3}
}
